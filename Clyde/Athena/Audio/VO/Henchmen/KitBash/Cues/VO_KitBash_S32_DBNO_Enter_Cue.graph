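# Sound cue for KitBash henchman voice lines that play on entering the DBNO (down but not out) state.
# A Random node picks one of five equally weighted wave players, each holding a variant of the line.
# It uses shared NPC alerted attenuation, the non-critical enemy VO sound class and the henchmen combat-chatter concurrency set, with a slight volume boost.
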SoundCue /Clyde/Athena/Audio/VO/Henchmen/KitBash/Cues/VO_KitBash_S32_DBNO_Enter_Cue
root=Random_0
nodes [
 {id=Random_0, children [WavePlayer_0,WavePlayer_1,WavePlayer_2,WavePlayer_3,WavePlayer_4], properties {weights [1,1,1,1,1]}}
 {id=WavePlayer_0, properties {soundWaveAssetPtr=/Clyde/Athena/Audio/VO/Henchmen/KitBash/Waves/VO_KitBash_S32_DBNO_Enter_010}}
 {id=WavePlayer_1, properties {soundWaveAssetPtr=/Clyde/Athena/Audio/VO/Henchmen/KitBash/Waves/VO_KitBash_S32_DBNO_Enter_011}}
 {id=WavePlayer_2, properties {soundWaveAssetPtr=/Clyde/Athena/Audio/VO/Henchmen/KitBash/Waves/VO_KitBash_S32_DBNO_Enter_012}}
 {id=WavePlayer_3, properties {soundWaveAssetPtr=/Clyde/Athena/Audio/VO/Henchmen/KitBash/Waves/VO_KitBash_S32_DBNO_Enter_013}}
 {id=WavePlayer_4, properties {soundWaveAssetPtr=/Clyde/Athena/Audio/VO/Henchmen/KitBash/Waves/VO_KitBash_S32_DBNO_Enter_014}}
]
properties {
 attenuationSettings=/Game/Sounds/Fort_Human_VO_BR/NPCs/Shared/NPC_VO_Alerted_Att
 maxDistance=12500
 soundClassObject=/Game/Sounds/FortSoundClassesAndModes/Fort_VO_Enemy_NonCritical
 volumeMultiplier=1.1147
 concurrencySet [/Game/Sounds/Fort_Human_VO_BR/NPCs/Shared/NPC_Henchmen_Combat_Chatter_Conc]
}
