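# Sound cue for the KitBash henchman's idle patrol voice lines.
# A single Random node picks with equal weight among eight wave players, mixing patrol-specific lines with two generic idle lines.
# It plays at reduced volume through the non-critical enemy VO sound class, using shared close-range NPC attenuation and the henchmen combat chatter concurrency group.
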SoundCue /Clyde/Athena/Audio/VO/Henchmen/KitBash/Cues/VO_KitBash_S32_IdlePatrol_Cue
root=Random_0
nodes [
 {
  id=Random_0
  children [WavePlayer_0,WavePlayer_1,WavePlayer_2,WavePlayer_3,WavePlayer_4,WavePlayer_5,WavePlayer_6,WavePlayer_7]
  properties {weights [1,1,1,1,1,1,1,1]}
 }
 {id=WavePlayer_0, properties {soundWaveAssetPtr=/Clyde/Athena/Audio/VO/Henchmen/KitBash/Waves/VO_KitBash_S32_IdleGeneric_012}}
 {id=WavePlayer_1, properties {soundWaveAssetPtr=/Clyde/Athena/Audio/VO/Henchmen/KitBash/Waves/VO_KitBash_S32_IdleGeneric_050}}
 {id=WavePlayer_2, properties {soundWaveAssetPtr=/Clyde/Athena/Audio/VO/Henchmen/KitBash/Waves/VO_KitBash_S32_IdlePatrol_010}}
 {id=WavePlayer_3, properties {soundWaveAssetPtr=/Clyde/Athena/Audio/VO/Henchmen/KitBash/Waves/VO_KitBash_S32_IdlePatrol_011}}
 {id=WavePlayer_4, properties {soundWaveAssetPtr=/Clyde/Athena/Audio/VO/Henchmen/KitBash/Waves/VO_KitBash_S32_IdlePatrol_020}}
 {id=WavePlayer_5, properties {soundWaveAssetPtr=/Clyde/Athena/Audio/VO/Henchmen/KitBash/Waves/VO_KitBash_S32_IdlePatrol_031}}
 {id=WavePlayer_6, properties {soundWaveAssetPtr=/Clyde/Athena/Audio/VO/Henchmen/KitBash/Waves/VO_KitBash_S32_IdlePatrol_041}}
 {id=WavePlayer_7, properties {soundWaveAssetPtr=/Clyde/Athena/Audio/VO/Henchmen/KitBash/Waves/VO_KitBash_S32_IdlePatrol_051}}
]
properties {
 attenuationSettings=/Game/Sounds/Fort_Human_VO_BR/NPCs/Shared/NPC_VO_Close_Att
 maxDistance=8500
 soundClassObject=/Game/Sounds/FortSoundClassesAndModes/Fort_VO_Enemy_NonCritical
 volumeMultiplier=0.4743
 concurrencySet [/Game/Sounds/Fort_Human_VO_BR/NPCs/Shared/NPC_Henchmen_Combat_Chatter_Conc]
}
