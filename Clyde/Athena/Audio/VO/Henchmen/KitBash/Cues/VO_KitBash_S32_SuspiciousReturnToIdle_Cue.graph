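# Sound cue for KitBash henchman voice lines, named for returning to idle from a suspicious state.
# A root Random node picks one of nine equally weighted wave variations.
# It uses shared NPC close-range VO attenuation, the non-critical enemy VO sound class and a reduced volume.
# A shared henchmen combat chatter concurrency set governs how these lines overlap with other chatter.
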SoundCue /Clyde/Athena/Audio/VO/Henchmen/KitBash/Cues/VO_KitBash_S32_SuspiciousReturnToIdle_Cue
root=Random_0
nodes [
 {
  id=Random_0
  children [WavePlayer_0,WavePlayer_1,WavePlayer_2,WavePlayer_3,WavePlayer_4,WavePlayer_5,WavePlayer_6,WavePlayer_7,WavePlayer_8]
  properties {weights [1,1,1,1,1,1,1,1,1]}
 }
 {id=WavePlayer_0, properties {soundWaveAssetPtr=/Clyde/Athena/Audio/VO/Henchmen/KitBash/Waves/VO_KitBash_S32_SuspiciousReturnToIdle_010}}
 {id=WavePlayer_1, properties {soundWaveAssetPtr=/Clyde/Athena/Audio/VO/Henchmen/KitBash/Waves/VO_KitBash_S32_SuspiciousReturnToIdle_020}}
 {id=WavePlayer_2, properties {soundWaveAssetPtr=/Clyde/Athena/Audio/VO/Henchmen/KitBash/Waves/VO_KitBash_S32_SuspiciousReturnToIdle_030}}
 {id=WavePlayer_3, properties {soundWaveAssetPtr=/Clyde/Athena/Audio/VO/Henchmen/KitBash/Waves/VO_KitBash_S32_SuspiciousReturnToIdle_031}}
 {id=WavePlayer_4, properties {soundWaveAssetPtr=/Clyde/Athena/Audio/VO/Henchmen/KitBash/Waves/VO_KitBash_S32_SuspiciousReturnToIdle_032}}
 {id=WavePlayer_5, properties {soundWaveAssetPtr=/Clyde/Athena/Audio/VO/Henchmen/KitBash/Waves/VO_KitBash_S32_SuspiciousReturnToIdle_040}}
 {id=WavePlayer_6, properties {soundWaveAssetPtr=/Clyde/Athena/Audio/VO/Henchmen/KitBash/Waves/VO_KitBash_S32_SuspiciousReturnToIdle_041}}
 {id=WavePlayer_7, properties {soundWaveAssetPtr=/Clyde/Athena/Audio/VO/Henchmen/KitBash/Waves/VO_KitBash_S32_SuspiciousReturnToIdle_050}}
 {id=WavePlayer_8, properties {soundWaveAssetPtr=/Clyde/Athena/Audio/VO/Henchmen/KitBash/Waves/VO_KitBash_S32_SuspiciousReturnToIdle_051}}
]
properties {
 attenuationSettings=/Game/Sounds/Fort_Human_VO_BR/NPCs/Shared/NPC_VO_Close_Att
 maxDistance=8500
 soundClassObject=/Game/Sounds/FortSoundClassesAndModes/Fort_VO_Enemy_NonCritical
 volumeMultiplier=0.7166
 concurrencySet [/Game/Sounds/Fort_Human_VO_BR/NPCs/Shared/NPC_Henchmen_Combat_Chatter_Conc]
}
